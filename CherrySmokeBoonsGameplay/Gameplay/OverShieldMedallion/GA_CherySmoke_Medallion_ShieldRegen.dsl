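; Gameplay ability for the Mask Maker overshield medallion that regenerates the player's shield up to a data-driven cap.
; StartRegen runs on activation, applying a gameplay effect and healing cue when below cap.
; A threshold task and polling timer remove the effect at cap; damage removes it and restarts a cooldown timer.
; Reload tags send a non-authority widget VFX event, and a PIE-only watcher reschedules regen below cap.
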(event Ability|EventActivateAbility
  (Variables|Default|SetShieldRegenCap (ScalableFloat|GetValueatLevel (Variables|Default|GetRow_ShieldRegenCap)))
  (CallFunction|StartRegen)
  (Development|PrintString "Mask Maker Medallion Added")
  (EventDispatchers|AssignOnDamaged (CallFunction|GetAvatarasFortPlayerPawn) (delegate Custom|OnDamaged_Event))
  (bind (_asynctaskproxy _bmatchescomparison _currentvalue) (Ability|Tasks|WaitForAttributeChangeThreshold "(AttributeName=\"CurrentShield\",Attribute=/Script/FortniteGame.FortHealthSet:CurrentShield,AttributeOwner=\"/Script/CoreUObject.Class'/Script/FortniteGame.FortHealthSet'\")" "GreaterThanOrEqualTo" (Variables|Default|GetShieldRegenCap) false (CallFunction|GetAvatarasFortPlayerPawn))
    (:then
      (Variables|Default|SetAsync_WatchShieldThreshold _asynctaskproxy)
      (bind _asynctaskproxy_2 (Ability|Tasks|WaitGameplayTagAddToActor (CallFunction|GetAvatarasFortPlayerPawn) "(TagName=\"Gameplay.Action.Player.Reload\")")
        (:then
          (Variables|Default|SetAsyncActionA _asynctaskproxy_2)
          ; For Testing in PIE
          (if (Editor|InEditororPIE)
            (bind (_asynctaskproxy_3 _bmatchescomparison_2 _currentvalue_2) (Ability|Tasks|WaitForAttributeChangeThreshold "(AttributeName=\"CurrentShield\",Attribute=/Script/FortniteGame.FortHealthSet:CurrentShield,AttributeOwner=\"/Script/CoreUObject.Class'/Script/FortniteGame.FortHealthSet'\")" "LessThan" (Variables|Default|GetShieldRegenCap) false (CallFunction|GetAvatarasFortPlayerPawn))
              (:then
                (Variables|Default|SetAsync_WatchShieldThreshold_LessThan _asynctaskproxy_3))
              (:OnChange
                (Development|PrintString (Utilities|String|Append "Current Shield Value: " (Utilities|String|ToString(Float) _currentvalue_2)))
                (if _bmatchescomparison_2
                  (if (not (Utilities|Time|IsTimerActivebyHandle (Variables|Default|GetRegenTimer)))
                    (CallFunction|EnableShieldRegen)))))))
        (:Added
          (CallFunction|ReloadWidgetVFX))))
    (:OnChange
      (Development|PrintString (Utilities|String|Append "Current Shield Value: " (Utilities|String|ToString(Float) _currentvalue)))
      (if _bmatchescomparison
        (CallFunction|EndRegen)))))

(event Custom|EnableShieldRegen
  ; For Testing in PIE
  (Development|PrintString "Manually Starting")
  (CallFunction|EndRegen)
  (Utilities|Time|ClearandInvalidateTimerbyHandle (Variables|Default|GetRegenTimer))
  (bind _returnvalue (Utilities|Time|SetTimerbyEvent (EventDispatchers|CreateEvent :!event Custom|StartRegen) (ScalableFloat|GetValueatLevel (Variables|Default|GetRow_ShieldRegenCooldown)) false false (ScalableFloat|GetValueatLevel (Variables|Default|GetRow_ShieldRegenCooldown))))
  (Variables|Default|SetRegenTimer _returnvalue))

(event Custom|OnDamaged_Event
  (params (DamagedActor :type Actor%20Object%20Reference) (Damage :type Float%20%28single-precision%29) (InstigatedBy :type Controller%20Object%20Reference) (DamageCauser :type Actor%20Object%20Reference) (HitLocation :type Vector) (FHitComponent :type Primitive%20Component%20Object%20Reference) (BoneName :type Name) (Momentum :type Vector))
  (Development|PrintString "Damage Taken")
  (CallFunction|EndRegen)
  (Utilities|Time|ClearandInvalidateTimerbyHandle (Variables|Default|GetRegenTimer))
  (bind _returnvalue (Utilities|Time|SetTimerbyEvent (EventDispatchers|CreateEvent :!event Custom|StartRegen) (ScalableFloat|GetValueatLevel (Variables|Default|GetRow_ShieldRegenCooldown)) false false (ScalableFloat|GetValueatLevel (Variables|Default|GetRow_ShieldRegenCooldown))))
  (Variables|Default|SetRegenTimer _returnvalue))

(event Custom|ReloadWidgetVFX
  ;[0.8 0.8 0.8 0.5] Widget VFX
  (Ability|HasAuthority
    (:False
      (Ability|SendGameplayEventtoActor (CallFunction|GetAvatarasFortPlayerPawn) "(TagName=\"Ability.CherrySmoke.Medallions.Overshield.Active\")" (Utilities|Struct|MakeGameplayEventData "(TagName=\"Ability.CherrySmoke.Medallions.Overshield.Active\")" 0 0 0 0 "()" "(GameplayTags=)" "(GameplayTags=)" 3e+01)))))

(event Custom|StartRegen
  (bind (_bsuccessfullyfoundattribute _returnvalue) (Ability|Attribute|GetFloatAttribute (CallFunction|GetAvatarasFortPlayerPawn) "(AttributeName=\"CurrentShield\",Attribute=/Script/FortniteGame.FortHealthSet:CurrentShield,AttributeOwner=\"/Script/CoreUObject.Class'/Script/FortniteGame.FortHealthSet'\")"))
  (Development|PrintString "Starting Mask Maker Medallion Regen")
  (if (< _returnvalue (Variables|Default|GetShieldRegenCap))
    (bind _returnvalue_2 (Ability|ApplyGameplayEffectSpecToOwner (Ability|MakeOutgoingGameplayEffectSpec "/CherrySmokeBoonsGameplay/Gameplay/OverShieldMedallion/GE_CherrySmoke_Medallion_OverShield.GE_CherrySmoke_Medallion_OverShield_C")))
    (Variables|Default|SetGE_ShieldRegen _returnvalue_2)
    (bind _returnvalue_3 (Utilities|Time|SetTimerbyEvent (delegate Custom|CheckShieldCap) 0.05 true))
    (Variables|Default|SetRegenCapTimer _returnvalue_3)
    (Ability|ExecuteGameplayCueWithParamsOnOwner (Variables|Default|GetSiphonHealingGCTag) (Ability|GameplayCue|MakeGameplayCueParameters 1.0)) ; Adding healing FX only on success\npass in whether shields were healed for VFX update as Magnitude
  ))

(event Custom|CheckShieldCap
  (bind (_bsuccessfullyfoundattribute _returnvalue) (Ability|Attribute|GetFloatAttribute (CallFunction|GetAvatarasFortPlayerPawn) "(AttributeName=\"CurrentShield\",Attribute=/Script/FortniteGame.FortHealthSet:CurrentShield,AttributeOwner=\"/Script/CoreUObject.Class'/Script/FortniteGame.FortHealthSet'\")"))
  (if (>= _returnvalue (Variables|Default|GetShieldRegenCap))
    (CallFunction|EndRegen)))

(event Custom|EndRegen
  (Ability|RemoveGameplayEffectFromOwnerWithHandle (Variables|Default|GetGE_ShieldRegen))
  (Utilities|Time|ClearandInvalidateTimerbyHandle (Variables|Default|GetRegenCapTimer))
  (Utilities|Time|ClearandInvalidateTimerbyHandle (Variables|Default|GetRegenTimer)))

(event Ability|EventOnEndAbility
  (params (bWasCancelled :type Boolean))
  (Utilities|Time|ClearandInvalidateTimerbyHandle (Variables|Default|GetRegenTimer))
  (GameplayTasks|EndTask (Variables|Default|GetAsync_WatchShieldThreshold))
  (CallFunction|EndRegen)
  (EventDispatchers|UnbindEventfromOnDamaged (CallFunction|GetAvatarasFortPlayerPawn) (EventDispatchers|CreateEvent :!event Custom|OnDamaged_Event)))
